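# Niagara system for the Drof vehicle boost cosmetic, with ribbon trails and exhaust cones recolored by painted variant.
# NE_Boost spawns quality-scaled ribbons per port, drawn by four tapering ribbon renderers, with inherited source movement fading over life and backward velocity chosen by frontend preview.
# NE_Cone loops short two-group bursts of meshes and glow sprites on skeletal mesh sockets, offset by DTCarOffset, scaling up and fading per spawn group.
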NiagaraSystem /VehicleCosmetics/Boosts/Boost_Drof/FX/NS_Boost_Drof
dependencies {
 usedDataInterfaces [
  &NiagaraDataInterfaceCurve
  &NiagaraDataInterfaceVectorCurve
  &NiagaraDataInterfaceArrayFloat3
  &NiagaraDataInterfaceSkeletalMesh
 ]
 usedDynamicInputs []
 usedModules [
  &self:SystemSpawnScript.SystemScriptSource.SystemScriptGraph.NiagaraNodeAssignment_0.SetVariables_#1
  &SystemState
 ]
}
emitters {
 NE_Boost {
  emitterUpdateScript [
   &EmitterState(Inactive Response=::NewEnumerator1, Life Cycle Mode=::NewEnumerator1, Loop Duration=1, UseLoopDelay=false)
   &VC_FilteredSocketCount()
   &VC_RateQualityHandler(Epic Spawn Rate=40, High Spawn Rate=36, Low Spawn Rate=26, Use Custom Rate?=true)
   &VC_SpawnRibbon_v1(Number of Ribbons=$Emitter.PortCount, SpawnRate=$Emitter.Spawn Rate)
  ]
  particleSpawnScript [
   &VC_PaintedVariantsOverride(
    Black=<rgba 0.02 0.02 0.02 1>
    BurntSienna=<rgba 0.7 0.1998 0.0656 1>
    Cobalt=<rgba 0.1296 0.2575 0.72 1>
    Crimson=<rgba 0.84 0.0252 0.0388 1>
    DefaultColor=<rgba 0.018 0.0465 0.36 1>
    ForestGreen=<rgba 0.1326 0.78 0.1653 1>
    Grey=<rgba 0.034 0.034 0.034 1>
    Lime=<rgba 0.545 0.75 0.135 1>
    Orange=<rgba 0.75 0.332 0.09 1>
    Pink=<rgba 0.75 0.1125 0.665 1>
    Purple=<rgba 0.3994 0.1125 0.75 1>
    R1 Black:LinearColor=<rgba 0.002 0.002 0.002 1>
    R1 BurntSienna:LinearColor=<rgba 0.2 0.0693 0.052 1>
    R1 Cobalt:LinearColor=<rgba 0.008 0.0733 0.4 1>
    R1 Crimson:LinearColor=<rgba 0.45 0.0135 0.0208 1>
    R1 DefaultColor=<rgba 0.04 0.23 0.8 1>
    R1 ForestGreen:LinearColor=<rgba 0.0084 0.28 0.0265 1>
    R1 Grey:LinearColor=<rgba 0.06 0.06 0.06 1>
    R1 Lime:LinearColor=<rgba 0.2424 0.36 0.0072 1>
    R1 Orange:LinearColor=<rgba 0.49 0.1666 0.0049 1>
    R1 Pink:LinearColor=<rgba 0.4 0.032 0.3693 1>
    R1 Purple:LinearColor=<rgba 0.1853 0.032 0.4 1>
    R1 Saffron:LinearColor=<rgba 0.45 0.3465 0.036 1>
    R1 SkyBlue:LinearColor=<rgba 0.02 0.2607 0.4 1>
    R1 TitaniumWhite:LinearColor=<rgba 0.6 0.6 0.6 1>
    R2 Black:LinearColor=<rgba 0.038 0.038 0.038 1>
    R2 BurntSienna:LinearColor=<rgba 0.35 0.0933 0.042 1>
    R2 Cobalt:LinearColor=<rgba 0.09 0.192 0.6 1>
    R2 Crimson:LinearColor=<rgba 0.8 0.08 0.092 1>
    R2 DefaultColor:LinearColor=<rgba 0.5694 0.78 0.1482 1>
    R2 ForestGreen:LinearColor=<rgba 0.09 0.6 0.1155 1>
    R2 Grey:LinearColor=<rgba 0.12 0.12 0.12 1>
    R2 Lime:LinearColor=<rgba 0.49 0.7 0.07 1>
    R2 Orange:LinearColor=<rgba 0.65 0.2369 0.0455 1>
    R2 Pink:LinearColor=<rgba 0.6 0.06 0.51 1>
    R2 Purple:LinearColor=<rgba 0.276 0.06 0.6 1>
    R2 Saffron:LinearColor=<rgba 0.78 0.6084 0.0936 1>
    R2 SkyBlue:LinearColor=<rgba 0.06 0.438 0.6 1>
    R2 TitaniumWhite:LinearColor=<rgba 0.28 0.28 0.28 1>
    R3 Black:LinearColor=<rgba 0.8 0.8 0.8 1>
    R3 BurntSienna:LinearColor=<rgba 0.82 0.4373 0.246 1>
    R3 Cobalt:LinearColor=<rgba 0.38 0.57 0.95 1>
    R3 Crimson:LinearColor=<rgba 0.9 0.315 0.3345 1>
    R3 DefaultColor:LinearColor=<rgba 0.9 0.9 0.9 1>
    R3 ForestGreen:LinearColor=<rgba 0.528 0.88 0.5458 1>
    R3 Grey:LinearColor=<rgba 0.43 0.43 0.43 1>
    R3 Lime:LinearColor=<rgba 0.675 0.9 0.36 1>
    R3 Orange:LinearColor=<rgba 0.88 0.5196 0.3256 1>
    R3 Pink:LinearColor=<rgba 0.9 0.306 0.801 1>
    R3 Purple:LinearColor=<rgba 0.603 0.306 0.9 1>
    R3 Saffron:LinearColor=<rgba 0.9 0.8103 0.486 1>
    R3 SkyBlue:LinearColor=<rgba 0.36 0.72 0.9 1>
    R3 TitaniumWhite:LinearColor=<rgba 0.9 0.9 0.9 1>
    Saffron=<rgba 0.55 0.4368 0.0275 1>
    SkyBlue=<rgba 0.1125 0.5587 0.75 1>
    TitaniumWhite=<rgba 0.17 0.17 0.17 1>
    Use R1 Painted Overrides=true
    Use R2 Painted Overrides=true
    Use R3 Painted Overrides=true
   )
   &VC_InitializeRibbon_v1(
    Color=$Particles.Color
    Custom Offset=true
    Enable Skeletal Position=true
    Lifetime=1.4
    Local Vector Offset=&AddVectorToPosition(Position=<5 0 0>, Vector=$User.DTCarOffset)
    Offset Amount=1
    Ribbon Width=51
   )
   &AddVelocity(Velocity=&SelectVectorFromArray(
     Direct Array Index=&MakeCustomIntFromBool(Bool=$User.bIsFrontendPreview, True INT=1, False INT=0)
     Vector Selection Array=DataInterface() {propertyValues {FloatData [<-1200 0 0>, <-1400 1 1>]}}
    )) {module=AddVelocity001}
  ]
  particleUpdateScript [
   &ParticleState()
   &InheritSourceMovement(
    Applied Position Scale=&VectorFromFloat(Value=&FloatFromCurve(FloatCurve=DataInterface() {propertyValues {Curve {keys [<at 0 1 RCIM_Cubic>, {arriveTangent=-2.1678, interpMode=RCIM_Cubic, leaveTangent=-2.1678, tangentMode=RCTM_User, time=1, value=0}]}}}))
    ApplyForce=false
    ApplyPosition=true
    ApplyPositionFalloff=false
   )
   &SolveForcesAndVelocity(Clamp Velocity=false, Limit Acceleration=false)
   &VC_UpdateRibbon_v1(Enable Skeletal Position=true)
   &ScaleRibbonWidth(Ribbon Width Scale=&FloatFromCurve(FloatCurve=DataInterface() {propertyValues {Curve {
        defaultValue=1
        keys [
         {arriveTangent=-0.7199, interpMode=RCIM_Cubic, leaveTangent=-0.7199, tangentMode=RCTM_User, time=0.1, value=1}
         {interpMode=RCIM_Cubic, tangentMode=RCTM_User, time=1, value=0.2}
        ]
       }}}))
   /SetVariables(
    Particles.widthR1:NiagaraFloat=&Multiply_Float(
     A=50
     B=&FloatFromCurve(FloatCurve=DataInterface() {propertyValues {Curve {
         defaultValue=1
         keys [
          {arriveTangent=-0.7199, interpMode=RCIM_Cubic, leaveTangent=-0.7199, tangentMode=RCTM_User, time=0.1, value=1}
          {interpMode=RCIM_Cubic, tangentMode=RCTM_User, time=1, value=0.2}
         ]
        }}})
    )
    Particles.widthR2:NiagaraFloat=&Multiply_Float(
     A=48
     B=&FloatFromCurve(FloatCurve=DataInterface() {propertyValues {Curve {
         defaultValue=1
         keys [
          {arriveTangent=-0.7199, interpMode=RCIM_Cubic, leaveTangent=-0.7199, tangentMode=RCTM_User, time=0.1, value=1}
          {interpMode=RCIM_Cubic, tangentMode=RCTM_User, time=1, value=0.2}
         ]
        }}})
    )
    Particles.widthR3:NiagaraFloat=&Multiply_Float(
     A=54
     B=&FloatFromCurve(FloatCurve=DataInterface() {propertyValues {Curve {
         defaultValue=1
         keys [
          {arriveTangent=-0.7199, interpMode=RCIM_Cubic, leaveTangent=-0.7199, tangentMode=RCTM_User, time=0.1, value=1}
          {interpMode=RCIM_Cubic, tangentMode=RCTM_User, time=1, value=0.2}
         ]
        }}})
    )
   ) {isSetParameters=true}
  ]
  rendererClasses [&NiagaraRibbonRendererProperties]
  renderers [
   {rendererClass=&NiagaraRibbonRendererProperties}
   {rendererClass=&NiagaraRibbonRendererProperties, rendererIndex=1}
   {rendererClass=&NiagaraRibbonRendererProperties, rendererIndex=2}
   {rendererClass=&NiagaraRibbonRendererProperties, rendererIndex=3}
  ]
 }
 NE_Cone {
  emitterUpdateScript [
   &EmitterState(Inactive Response=::NewEnumerator1, Life Cycle Mode=::NewEnumerator1, Loop Duration=0.1, UseLoopDelay=false)
   &VC_FilteredSocketCount()
   &SpawnBurst_Instantaneous(Spawn Count=6, Use Spawn Probability=false, UseLoopCountLimit=false)
   &SpawnBurst_Instantaneous(Spawn Count=9, Spawn Group=1, Use Spawn Probability=false, UseLoopCountLimit=false) {module=SpawnBurst_Instantaneous001}
   &SpawnRate(SpawnRate=30, Use Spawn Probability=false) {enabled=false}
  ]
  particleSpawnScript [
   &VC_PaintedVariantsOverride(
    Black=<rgba 0.9 0.9 0.9 1>
    BurntSienna=<rgba 0.3 0.0673 0.015 1>
    Cobalt=<rgba 0.006 0.0648 0.3 1>
    Crimson=<rgba 0.3 0.006 0.0109 1>
    ForestGreen=<rgba 0.006 0.3 0.0207 1>
    Grey=<rgba 0.03 0.03 0.03 1>
    Lime=<rgba 0.202 0.3 0.006 1>
    Orange=<rgba 0.3 0.104 0.006 1>
    Pink=<rgba 0.3 0.006 0.251 1>
    Purple=<rgba 0.1285 0.006 0.3 1>
    R1 Black:LinearColor=<rgba 0.9 0.9 0.9 1>
    R1 BurntSienna:LinearColor=<rgba 0.95 0.57 0.38 1>
    R1 Cobalt:LinearColor=<rgba 0.38 0.57 0.95 1>
    R1 Crimson:LinearColor=<rgba 0.95 0.38 0.3895 1>
    R1 DefaultColor=<rgba 0.9 0.85 0.05 1>
    R1 ForestGreen:LinearColor=<rgba 0.38 0.95 0.4085 1>
    R1 Grey:LinearColor=<rgba 0.95 0.95 0.95 1>
    R1 Lime:LinearColor=<rgba 0.76 0.95 0.38 1>
    R1 Orange:LinearColor=<rgba 0.95 0.57 0.38 1>
    R1 Pink:LinearColor=<rgba 0.95 0.38 0.855 1>
    R1 Purple:LinearColor=<rgba 0.6175 0.38 0.95 1>
    R1 Saffron:LinearColor=<rgba 0.95 0.8075 0.38 1>
    R1 SkyBlue:LinearColor=<rgba 0.38 0.76 0.95 1>
    R1 TitaniumWhite:LinearColor=<rgba 0.9 0.9 0.9 1>
    R2 Black:LinearColor=<rgba 0.9 0.9 0.9 1>
    R2 BurntSienna:LinearColor=<rgba 0.95 0.6333 0.475 1>
    R2 Cobalt:LinearColor=<rgba 0.475 0.6333 0.95 1>
    R2 Crimson:LinearColor=<rgba 0.95 0.475 0.4829 1>
    R2 DefaultColor:LinearColor=<rgba 1 0.9959 0.5304 1>
    R2 ForestGreen:LinearColor=<rgba 0.475 0.95 0.4988 1>
    R2 Grey:LinearColor=<rgba 0.5 0.5 0.5 1>
    R2 Lime:LinearColor=<rgba 0.7917 0.95 0.475 1>
    R2 Orange:LinearColor=<rgba 0.95 0.6333 0.475 1>
    R2 Pink:LinearColor=<rgba 0.95 0.475 0.8708 1>
    R2 Purple:LinearColor=<rgba 0.6729 0.475 0.95 1>
    R2 Saffron:LinearColor=<rgba 0.95 0.8313 0.475 1>
    R2 SkyBlue:LinearColor=<rgba 0.475 0.7917 0.95 1>
    R2 TitaniumWhite:LinearColor=<rgba 0.9 0.9 0.9 1>
    R3 Black:LinearColor=<rgba 0.9 0.9 0.9 1>
    R3 BurntSienna:LinearColor=<rgba 0.95 0.6333 0.475 1>
    R3 Cobalt:LinearColor=<rgba 0.475 0.6333 0.95 1>
    R3 Crimson:LinearColor=<rgba 0.95 0.475 0.4829 1>
    R3 DefaultColor:LinearColor=<rgba 1 0.7829 0.3304 1>
    R3 ForestGreen:LinearColor=<rgba 0.475 0.95 0.4988 1>
    R3 Grey:LinearColor=<rgba 0.6 0.6 0.6 1>
    R3 Lime:LinearColor=<rgba 0.7917 0.95 0.475 1>
    R3 Orange:LinearColor=<rgba 0.95 0.6333 0.475 1>
    R3 Pink:LinearColor=<rgba 0.95 0.475 0.8708 1>
    R3 Purple:LinearColor=<rgba 0.6729 0.475 0.95 1>
    R3 Saffron:LinearColor=<rgba 0.95 0.8313 0.475 1>
    R3 SkyBlue:LinearColor=<rgba 0.475 0.7917 0.95 1>
    R3 TitaniumWhite:LinearColor=<rgba 0.6 0.6 0.6 1>
    Saffron=<rgba 0.3 0.2265 0.006 1>
    SkyBlue=<rgba 0.006 0.202 0.3 1>
    TitaniumWhite=<rgba 0.3 0.3 0.3 1>
    Use R1 Painted Overrides=true
    Use R2 Painted Overrides=true
    Use R3 Painted Overrides=true
   )
   &InitializeParticle(
    Color Mode=::NewEnumerator0
    Lifetime Max=0.15
    Lifetime Min=0.1
    Lifetime Mode=::NewEnumerator1
    Mesh Scale Mode=::NewEnumerator1
    Mesh Uniform Scale Max=2.5
    Mesh Uniform Scale Min=2
    Sprite Size Mode=::NewEnumerator1
    Uniform Sprite Size Max=90
    Uniform Sprite Size Min=70
    UsePositionOffset=false
    Write Material Random=true
   )
   &PartitionParticles(Number Of Partitions=$Emitter.PortCount)
   &SkeletalMeshLocation(
    Bone / Socket Index=$Particles.PartitionParticles.Partition
    Bone / Socket Sampling Mode=::NewEnumerator5
    MaskSpawnGroup=false
    Mesh Sampling Type=::NewEnumerator6
    Orientation Sampling=::NewEnumerator1
    Sampled Position Offset=&AddVectorToPosition(
     Position=&MergeHelper_VectorSwitch(0=<-40 0 0>, 1=<-45 0 0>, 2=?, 3=?, Index=&SpawnGroup(), Num=::NewEnumerator0)
     Vector=$User.DTCarOffset
    )
    Skeletal Mesh=$System.SkeletalMesh
   )
   &AddVelocity(Velocity=&RandomRangeVector(Minimum=<5 0 0>, Maximum=<20 0 0>), Velocity Speed Scale=-1) {enabled=false}
   &InitialMeshOrientation(
    Mesh Axis To Orient=<0 -1 0>
    Mesh Orientation Mode=::NewEnumerator2
    Orientation Vector=<-1 0 0>
    Rotation=&RandomRangeVector(Minimum=<0 0 0>, Maximum=<0 1 0>)
    Use Rotation Vector=true
   )
   /SetVariables(Particles.VisibilityTag=$Emitter.SpawnGroup) {isSetParameters=true}
  ]
  particleUpdateScript [
   &ParticleState()
   &SolveForcesAndVelocity(Clamp Velocity=false, Limit Acceleration=false) {enabled=false}
   &ScaleSpriteSize(Uniform Curve Sprite Scale=DataInterface() {propertyValues {Curve {keys [{tangentMode=RCTM_User, time=0, value=0.3}, {tangentMode=RCTM_User, time=1, value=1}]}}})
   &ScaleMeshSize(Scale Factor=&VectorFromCurve(
     VectorCurve=DataInterface() {propertyValues {XCurve {keys [<at 0 0.6>,<at 1 1.2>]}, YCurve {keys [<at 0 0.6>,<at 1 1.6>]}, ZCurve {keys [<at 0 0.6>,<at 1 1.2>]}}}
     Scale Curve=<1 0.7 1>
    ))
   &ScaleColor(
    Scale Alpha=&MergeHelper_FloatSwitch(
     0=&FloatFromCurve(FloatCurve=DataInterface() {propertyValues {Curve {
         defaultValue=0
         keys [
          {arriveTangent=-4.7618, interpMode=RCIM_Cubic, leaveTangent=-4.7618, tangentMode=RCTM_User, time=0.6, value=1}
          {arriveTangent=-0.4047, interpMode=RCIM_Cubic, leaveTangent=-0.4047, tangentMode=RCTM_User, time=1, value=0}
         ]
        }}})
     1=&FloatFromCurve(
      FloatCurve=DataInterface() {propertyValues {Curve {
         defaultValue=1
         keys [{interpMode=RCIM_Cubic, tangentMode=RCTM_User, time=0, value=1}, {interpMode=RCIM_Cubic, tangentMode=RCTM_User, time=1, value=0}]
        }}}
      Scale Curve=3
     )
     2=?
     3=?
     Index=&SpawnGroup()
     Num=::NewEnumerator0
    )
    Scale RGB=&VectorFromFloat(Value=1)
    ScaleA=true
    ScaleRGB=false
   )
   /SetVariables(
    Particles.glow2color:LinearColor=&MultiplyLinearColor(
     Linear Color=$Particles.R3 Color
     Scale Factor=&MakeVector4FromVectorAndFloat("Vector (RGB)"=&VectorFromFloat(Value=1), "Float (Alpha)"=0.15)
    )
    Particles.glow2position:NiagaraPosition=&AddVectorToPosition(Position=$Particles.Position, Vector=<-20 0 0>)
    Particles.glow2size:Vector2f=&Multiply_Vector2D_ByFloat(
     Float=&RandomRangeFloat(Minimum=120, Maximum=170)
     Vector2D=&Vector2DFromFloat(Value=&FloatFromCurve(FloatCurve=DataInterface() {propertyValues {Curve {keys [<at 0 1>,<at 1 1.2>]}}}))
    )
   ) {isSetParameters=true}
   /SetVariables(Particles.conescale:Vector3f=&Multiply_Vector(
     A=&VectorFromFloat(Value=&RandomRangeFloat(Minimum=1.8, Maximum=2.1))
     B=&VectorFromCurve(
      VectorCurve=DataInterface() {propertyValues {XCurve {keys [<at 0 0.6>,<at 1 1.2>]}, YCurve {keys [<at 0 0.6>,<at 1 1.6>]}, ZCurve {keys [<at 0 0.6>,<at 1 1.2>]}}}
      Scale Curve=<0.9 0.6 0.9>
     )
    )) {isSetParameters=true}
  ]
  rendererClasses [&NiagaraMeshRendererProperties, &NiagaraSpriteRendererProperties]
  renderers [
   {rendererClass=&NiagaraMeshRendererProperties, rendererIndex=0}
   {rendererClass=&NiagaraMeshRendererProperties, rendererIndex=1}
   {rendererClass=&NiagaraSpriteRendererProperties, rendererIndex=2}
   {rendererClass=&NiagaraSpriteRendererProperties, rendererIndex=3}
  ]
 }
}
summary {
 systemName=NS_Boost_Drof
 emitters [
  {emitterName=NE_Cone, rendererClasses [&NiagaraMeshRendererProperties, &NiagaraSpriteRendererProperties]}
  {emitterName=NE_Boost, rendererClasses [&NiagaraRibbonRendererProperties]}
 ]
 userVariables [
  User.bIsFrontendPreview {type=NiagaraBool}
  User.bIsPainted {type=NiagaraBool}
  User.ColorVariantID {type=NiagaraInt32}
  User.DTCarOffset {type=Vector3f, description="", defaultValue {struct=Vector3f, value=<0 0 0>}}
  User.PaintedColor {type=LinearColor, description="", defaultValue {struct=LinearColor, value=<rgba 1 1 1 1>}}
  User.PaintedScalar {type=NiagaraFloat, defaultValue=1, description=""}
  User.PortCount {type=NiagaraInt32, defaultValue=2}
 ]
}
